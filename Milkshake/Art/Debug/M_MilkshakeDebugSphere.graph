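# M_MilkshakeDebugSphere is a translucent, two-sided Default Lit material, named for tinting debug spheres in the Milkshake content.
# It exposes a DebugSphereColor vector parameter, defaulting to pink-magenta, and an Opacity scalar defaulting to half, so instances can adjust both.
# The graph holds only these two parameter nodes, with no math or textures and no connections to material outputs shown.
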Material /Milkshake/Art/Debug/M_MilkshakeDebugSphere
graph {
 properties {blendMode=BLEND_Translucent, pixelDepthOffsetMode=PDOM_Legacy, twoSided=true, usedShadingModels=Default Lit}
 nodes [
  ScalarParameter() {defaultValue=0.5, parameterName=Opacity, sliderMax=1, sortPriority=1}
  VectorParameter() {parameterName=DebugSphereColor, sortPriority=0, defaultValue=<rgba 1 0 0.7892 0>}
 ]
}
